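# Animation Blueprint for the QuartzSqueak male player head cosmetic that signals eye blink changes to the head's materials.
# After calling the parent update, it compares the left and right blink pose curves against values stored from the previous update.
# It writes the HasEyeBlinkChanged scalar on the owning mesh's materials, set when either curve differs, then caches both values in local temporary variables instead of member variables.
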AnimBlueprint /BRCosmetics/Characters/Player/Male/Medium/Heads/M_MED_QuartzSqueak_Head/Meshes/M_MED_QuartzSqueak_Head_AnimBP
target_skeleton=/Game/Characters/Player/Common/Fortnite_Base_Head/Fortnite_Base_Head_Export_Skeleton
graphs [EventGraph {nodes [
   n1 = Event BlueprintUpdateAnimation()
   > Parent: BlueprintUpdateAnimation!(execute, DeltaTimeX:float=@n1.DeltaTimeX) -> (then) {is=title K2Node_CallParentFunction}
   n2 = GetCurveValue(CurveName=R_blink_pose)
   n3 = Local Wildcard!() -> (Variable:"Float (single-precision) (by ref)") {is=title K2Node_TemporaryVariable}
   n4 = "Not Equal (Float)"(A=@n2, B=@n3)
   n5 = GetCurveValue(CurveName=L_blink_pose)
   n6 = Local Wildcard()
   n7 = "Not Equal (Float)"(A=@n5, B=@n6)
   > SetScalarParameterValueOnMaterials!(
    execute
    self:Mesh Component Object Reference=GetOwningComponent()
    ParameterName:Name=HasEyeBlinkChanged
    ParameterValue:float="To Float (Boolean)"(InBool=OR Boolean(A=@n7, B=@n4))
   ) -> (then)
   > "Set Float (single-precision) (by ref)"!(execute:Exec, Target:"Float (single-precision) (by ref)"=@n6, Value:float=@n5) -> (then:Exec) {is=title K2Node_VariableSetRef}
   > "Set Float (single-precision) (by ref)"(Target=@n3, Value=@n2)
   TryGetPawnOwner()
  ]}]
variables []
